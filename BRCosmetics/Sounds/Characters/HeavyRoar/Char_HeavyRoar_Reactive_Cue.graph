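# Sound cue for the HeavyRoar character cosmetic's reactive sound, played as player foley.
# A random node picks one of three equally weighted Reactive wave variants.
# It routes to the Fort player foley sound class with basic foley attenuation and a boosted volume.
# Overridden concurrency allows one instance per owner, stopping the oldest when a new one plays.
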SoundCue /BRCosmetics/Sounds/Characters/HeavyRoar/Char_HeavyRoar_Reactive_Cue
root=Random_1
nodes [
 {id=Random_1, children [WavePlayer_3,WavePlayer_4,WavePlayer_5], properties {weights [1,1,1]}}
 {id=WavePlayer_3, properties {soundWaveAssetPtr=&/Char_HeavyRoar_Reactive_03}}
 {id=WavePlayer_4, properties {soundWaveAssetPtr=&/Char_HeavyRoar_Reactive_01}}
 {id=WavePlayer_5, properties {soundWaveAssetPtr=&/Char_HeavyRoar_Reactive_02}}
]
properties {
 attenuationSettings=&Foley_Attenuation_Basic_Att
 bOverrideConcurrency=true
 maxDistance=1536
 soundClassObject=&Fort_Player_Foley
 volumeMultiplier=3.5
 concurrencyOverrides {bLimitToOwner=true, maxCount=1, resolutionRule=StopOldest}
}
